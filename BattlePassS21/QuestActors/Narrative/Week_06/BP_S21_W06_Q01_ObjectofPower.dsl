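; Season 21 Week 6 narrative quest actor for the Object of Power, a child Blueprint that adjusts its state based on the player's quest progress.
; When a player enters cosmetic range, it runs the parent logic, then checks for a valid controller and that it is not already hidden.
; If either the Q03 S02 or S03 objective is active, it enables hide-after-interact and calls ChangeStateforPlayer for that player.
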(event EventBeginPlay
  (|CallParentFunction) ; This node is disabled and will not be called.\nDrag off pins to build functionality.
)

(event Collision|EventActorBeginOverlap
  (params (OtherActor :type Actor%20Object%20Reference))
  (|CallParentFunction OtherActor) ; This node is disabled and will not be called.\nDrag off pins to build functionality.
)

(event EventTick
  (params (DeltaSeconds :type Float%20%28single-precision%29))
  (|CallParentFunction DeltaSeconds) ; This node is disabled and will not be called.\nDrag off pins to build functionality.
)

(event EventOnEnterCosmetic
  (params (PlayerController :type Fort%20Player%20Controller%20Object%20Reference) (PlayerPawn :type Fort%20Player%20Pawn%20Athena%20Object%20Reference))
  (|CallParentFunction PlayerController PlayerPawn)
  (if (Utilities|IsValid PlayerController :!class /Script/Engine.KismetSystemLibrary)
    (if (Variables|Rendering|GetbHidden)
      _
      (elif (or (Quests|IsQuestObjectiveActive (Quests|GetQuestManager PlayerController "Invalid") (Variables|Default|GetObj_W06_Q03_S02)) (Quests|IsQuestObjectiveActive (Quests|GetQuestManager PlayerController "Invalid") (Variables|Default|GetObj_W06_Q03_S03)))
        (Variables|Design|PostInteraction|SetbHideAfterInteract true)
        (CallFunction|ChangeStateforPlayer PlayerController)))))

(fn ConstructionScript
  (|CallParentFunction))
