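# Niagara system for the idle look of a shield in the Season 21 Week 6 narrative quest VFX.
# A single Shield emitter burst-spawns particles with an HDR violet color and disables lifetime kill, so they persist.
# Two mesh renderers draw the shield alongside a light renderer.
# Each update derives the light color from particle color with reduced red and a boosted alpha.
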NiagaraSystem /BattlePassS21/QuestActors/Narrative/Week_06/VFX/M_S21_W06_Shield_Idle
dependencies {usedDataInterfaces [], usedDynamicInputs [], usedModules [&SystemState]}
emitters {Shield {
  emitterUpdateScript [
   &EmitterState(
    Inactive Response=::NewEnumerator1
    Life Cycle Mode=::NewEnumerator1
    Loop Behavior=::NewEnumerator1
    Loop Duration Mode=::NewEnumerator1
    UseLoopDelay=false
   )
   &SpawnBurst_Instantaneous(Use Spawn Probability=false, UseLoopCountLimit=false)
  ]
  particleSpawnScript [&InitializeParticle(Color=<rgba 1.228 0 3 1>, Lifetime=5, UsePositionOffset=false, Write Material Random=true)]
  particleUpdateScript [
   &ParticleState(Kill Particles When Lifetime Has Elapsed=false)
   /SetVariables(Particles.LightColor=&MultiplyLinearColor(
     Linear Color=$Particles.Color
     Scale Factor=&MakeFloat4FromVectorAndFloat("Vector (RGB)"=<0.333 1 1>, "Float (Alpha)"=22)
    )) {isSetParameters=true}
  ]
  rendererClasses [&NiagaraMeshRendererProperties, &NiagaraLightRendererProperties]
  renderers [
   {rendererClass=&NiagaraMeshRendererProperties, rendererIndex=0}
   {rendererClass=&NiagaraMeshRendererProperties, rendererIndex=1}
   {rendererClass=&NiagaraLightRendererProperties, rendererIndex=2}
  ]
 }}
summary {
 systemName=M_S21_W06_Shield_Idle
 emitters [{emitterName=Shield, rendererClasses [&NiagaraMeshRendererProperties, &NiagaraLightRendererProperties]}]
 userVariables []
}
